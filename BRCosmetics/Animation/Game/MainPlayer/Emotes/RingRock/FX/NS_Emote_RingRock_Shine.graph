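# Niagara system for the RingRock emote shine, producing short-lived sprite glints tinted by time of day.
# The Blings emitter combines instant bursts with a probabilistic spawn rate, split into two spawn groups that switch lifetime, size and spin.
# Particles write the group as a visibility tag for four sprite renderers, carry ring and dark-glow colors, and grow while curves flash and fade their alpha.
# The Rays emitter lists no renderers.
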NiagaraSystem /BRCosmetics/Animation/Game/MainPlayer/Emotes/RingRock/FX/NS_Emote_RingRock_Shine
dependencies {
 usedDataInterfaces [&NiagaraDataInterfaceCurve]
 usedDynamicInputs [&RandomRangeLinearColor]
 usedModules [
  &SystemState
  &self:SystemSpawnScript.SystemScriptSource.SystemScriptGraph.NiagaraNodeAssignment_0.SetVariables_#1
 ]
}
emitters {
 Blings {
  emitterUpdateScript [
   &EmitterState()
   &SpawnBurst_Instantaneous(Spawn Count=3, Use Spawn Probability=false, UseLoopCountLimit=false)
   &SpawnRate(Spawn Probability=0.65, SpawnGroup=1, SpawnRate=5, Use Spawn Probability=true)
   &SpawnBurst_Instantaneous(Spawn Count=2, Spawn Group=1, Spawn Time=0.18, Use Spawn Probability=false, UseLoopCountLimit=false) {module=SpawnBurst_Instantaneous001}
  ]
  particleSpawnScript [
   &InitializeParticle(
    Color=&FNTimeOfDayColor(
     Night Color=<rgba 1.6 1.5232 1.1917 1>
     Morning Color=<rgba 1.8 1.7277 1.4156 1>
     Day Color=<rgba 2 1.9197 1.5729 1>
     Evening Color=<rgba 1.8 1.7277 1.4156 1>
    )
    Lifetime=&MergeHelper_FloatSwitch(
     0=0.24
     1=&RandomRangeFloat(Minimum=0.13, Maximum=0.2)
     2=?
     3=?
     Index=&SpawnGroup()
     Num=::NewEnumerator0
    )
    Sprite Rotation Angle=25
    Sprite Rotation Mode=::NewEnumerator0
    Sprite Size Mode=::NewEnumerator3
    Uniform Sprite Size=&MergeHelper_FloatSwitch(
     0=39
     1=&RandomRangeFloat(Minimum=20, Maximum=35)
     2=?
     3=?
     Index=&SpawnGroup()
     Num=::NewEnumerator0
    )
    UsePositionOffset=false
    Write Material Random=true
   )
   &CameraOffset(Camera Offset Amount=25)
   /SetVariables(Particles.Ring Color:LinearColor=<rgba 1 0.9598 0.7865 0.5>) {isSetParameters=true}
   /SetVariables(Particles.Dark Glow Color:LinearColor=<rgba 0.0156 0.0156 0.0156 0.2>) {isSetParameters=true}
   /SetVariables(Particles.VisibilityTag=&SpawnGroup()) {isSetParameters=true}
  ]
  particleUpdateScript [
   &ParticleState()
   &SpriteRotationRate(Rotation Rate=&MergeHelper_FloatSwitch(0=-650, 1=0, 2=?, 3=?, Index=&SpawnGroup(), Num=::NewEnumerator0))
   &ScaleSpriteSize(Uniform Curve Sprite Scale=DataInterface() {propertyValues {Curve {keys [{tangentMode=RCTM_User, time=0, value=0.5}, {tangentMode=RCTM_User, time=1.1, value=1.1}]}}})
   &ScaleColor(
    Scale Alpha=&FloatFromCurve(
     FloatCurve=DataInterface() {propertyValues {Curve {keys [<at 0 0 RCIM_Cubic>, <at 0.0528 1.2 RCIM_Cubic>, {arriveTangentWeight=0.8253, interpMode=RCIM_Cubic, tangentMode=RCTM_User, time=1, value=0}]}}}
     Scale Curve=&RandomRangeFloat(Minimum=0.3, Maximum=1)
    )
    ScaleA=true
    ScaleRGB=true
   )
   /SetVariables(Particles.Ring Color:LinearColor=&MultiplyLinearColor(
     Linear Color=$Particles.Initial.Ring Color
     Scale Factor=&MakeVector4FromVectorAndFloat(
      "Vector (RGB)"=<1 1 1>
      "Float (Alpha)"=&FloatFromCurve(
       FloatCurve=DataInterface() {propertyValues {Curve {keys [
           <at 0 0 RCIM_Cubic>
           <at 0.0528 1.2 RCIM_Cubic>
           {arriveTangent=-2.7773, interpMode=RCIM_Cubic, leaveTangent=-2.7773, tangentMode=RCTM_User, time=0.9, value=0}
          ]}}}
       Scale Curve=&RandomRangeFloat(Minimum=0.3, Maximum=1)
      )
     )
    )) {isSetParameters=true}
  ]
  rendererClasses [&NiagaraSpriteRendererProperties]
  renderers [
   {rendererClass=&NiagaraSpriteRendererProperties}
   {rendererClass=&NiagaraSpriteRendererProperties, rendererIndex=1}
   {rendererClass=&NiagaraSpriteRendererProperties, rendererIndex=2}
   {rendererClass=&NiagaraSpriteRendererProperties, rendererIndex=3}
  ]
 }
 Rays {rendererClasses [], renderers []}
}
summary {
 systemName=NS_Emote_RingRock_Shine
 emitters [{emitterName=Blings, rendererClasses [&NiagaraSpriteRendererProperties]}, {emitterName=Rays, rendererClasses []}]
 userVariables []
}
